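# Material function that layers animated lightning emissive onto the DriftTrooper backpack's input material attributes.
# Sine waves with per-step random amplitude distort a bump-offset mask, whose smoothstepped RGB channels cycle over time as three bolt shapes.
# Tinted bolts fade with WeaponSwitchTime and combine with a panning electric-noise Fresnel background.
# The result is channel-masked, dimmed at evening and night, and added to emissive, while base color lerps toward black.
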MaterialFunction /Game/Accessories/FORT_Backpacks/Backpack_M_MED_DriftTrooper/Materials/MF_DriftTrooperLightning
graph {nodes [
  n1 = GetMaterialAttributes(None=FunctionInput() {inputName=Material, inputType=FunctionInput_MaterialAttributes}) {attributeGetTypes [#1, #2]}
  n2 = NamedRerouteDeclaration(None=@n1.EmissiveColor) {name=EmissiveIn, nodeColor=<rgba 1 0 0.8437 1>}
  n3 = ChannelMaskParameter(None=FunctionInput() {bUsePreviewValueAsDefault=true, inputName=Tech Art Mask In, previewValue=<1 1 1 1>}) {group=Lightning, maskChannel=Green, parameterName=LightningEmissiveMask, defaultValue=<rgba 0 1 0 0>}
  n4 = NamedRerouteDeclaration(None=VectorParameter() {group=Lightning, parameterName=EmissiveBaseTint, defaultValue=<rgba 0.3 0.026 0.1433 1>}.RGB) {name=EmissiveColorInput, nodeColor=<rgba 1 0.2109 0 1>}
  n5 = ScalarParameter() {defaultValue=60, group=Lightning, parameterName=LightningIntensity}
  n6 = NamedRerouteUsage()
  n7 = ComponentMask(None=@n6) {g=true}
  n8 = Multiply(A=Time(), B=ScalarParameter() {defaultValue=0.4, group=Lightning, parameterName=WaveSpeed}) {constB=0.1}
  n9 = Multiply(A=Frac(None=Add(A=@n7, B=@n8)), B=ScalarParameter() {defaultValue=10, group=Lightning, parameterName=WaveFrequency}) {constB=20}
  n10 = AppendVector(A=Sine(None=@n9), B=Constant())
  n11 = Constant()
  n12 = Floor(None=@n9)
  n13 = ComponentMask(None=MaterialFunctionCall!(UVIndexInput=AppendVector(A=@n11, B=@n12)) {materialFunction=&/MF_RandomNoise}.Result) {r=true}
  n14 = Subtract(A=@n7) {constB=0.5}
  n15 = Multiply(
   A=Multiply(A=Multiply(A=@n10, B=@n13), B=ScalarParameter() {defaultValue=0.03, group=Lightning, parameterName=WaveIntensity})
   B=SphereMask(A=Abs(None=@n14), B=Constant()) {attenuationRadius=0.5, hardnessPercent=0}
  )
  n16 = TextureSampleParameter2D(UVs=BumpOffset(Coordinate=Add(A=@n6, B=@n15), Height=ScalarParameter() {defaultValue=-0.05, group=Lightning, parameterName=BumpHeightTiny})) {
   group=Lightning
   parameterName=LightningPattern
   samplerType=SAMPLERTYPE_Masks
   texture=/Game/Accessories/FORT_Backpacks/Backpack_M_MED_DriftTrooper/Textures/T_LightningMask_Distort
  }
  n17 = Multiply(
   A=MaterialFunctionCall!(UV_Coord=@n6) {materialFunction=&/MF_UV01_Mask}.Mask
   B=@n16.RGB
  )
  n18 = MaterialFunctionCall(Float3=SmoothStep(
    Min=ScalarParameter() {defaultValue=0.5, group=Lightning, parameterName=SDF_Min}
    Max=ScalarParameter() {defaultValue=0.9, group=Lightning, parameterName=SDF_Max}
    Value=@n17
   ) {constMin=0.6}) {materialFunction=&BreakOutFloat3Components}
  n19 = ScalarParameter() {defaultValue=2, group=Lightning, parameterName=LightningDivisor}
  n20 = MaterialFunctionCall!(Time, Divisor=@n19, Mod=Constant() {r=1}) {materialFunction=/Game/Effects/Fort_Effects/Materials/_Functions/MF_ModuloTime}
  n21 = MaterialFunctionCall(
   A=NamedRerouteDeclaration(None=@n18.R) {name=Lightning01, nodeColor=<rgba 0 1 0.1797 1>}
   B=NamedRerouteDeclaration(None=@n18.G) {name=Lightning02, nodeColor=<rgba 0 1 0.1797 1>}
   C=NamedRerouteDeclaration(None=@n18.B) {name=Lightning03, nodeColor=<rgba 0 1 0.1797 1>}
   Alpha=ConstantBiasScale(None=Sine(None=@n20.Clamped 0-1))
  ) {materialFunction=&Lerp_3Color}
  n22 = Multiply(A=ScalarParameter() {defaultValue=0.3, group=Lightning, parameterName=RadiusEmissiveIntensity}, B=NamedRerouteUsage()) {constA=0.1}
  n23 = TextureSampleParameter2D() {
   group=Lightning
   parameterName=LightningPatternArea
   samplerType=SAMPLERTYPE_LinearGrayscale
   texture=/Game/Accessories/FORT_Backpacks/Backpack_M_MED_DriftTrooper/Textures/T_LightningMask_TechArt_Area
  }
  n24 = Multiply(
   A=Multiply(A=Multiply(A=@n4, B=@n5) {constB=50}, B=Add(A=@n21.Result, B=@n22))
   B=Power(Base=@n23.A, Exp=ScalarParameter() {defaultValue=3, group=Lightning, parameterName=AlphaPower}) {constExponent=3}
  )
  n25 = Time()
  n26 = ScalarParameter() {parameterName=WeaponSwitchTime}
  n27 = Saturate(None=Divide(A=Subtract(A=@n25, B=@n26), B=ScalarParameter() {defaultValue=0.5, parameterName=SwitchTimeScale}))
  n28 = LinearInterpolate(A=OneMinus(None=@n27), B=@n27, Alpha=ScalarParameter() {defaultValue=1, group=Lightning, parameterName=CCPM_Lightning})
  n29 = ScalarParameter() {defaultValue=0.5, group=Lightning, parameterName=BackGroundeEmissivIntensity}
  n30 = ScalarParameter() {defaultValue=2, group=Lightning, parameterName=UVScale}
  n31 = TextureCoordinate()
  n32 = ScalarParameter() {group=Lightning, parameterName=PatternSpeed_X, sortPriority=2}
  n33 = ScalarParameter() {defaultValue=0.2, group=Lightning, parameterName=PatternSpeed_Y, sortPriority=2}
  n34 = TextureSampleParameter2D(UVs=Panner(Coordinate=Multiply(A=@n30, B=@n31), Speed=AppendVector(A=@n32, B=@n33))) {
   group=Lightning
   parameterName=ElectricNoisePattern
   samplerType=SAMPLERTYPE_LinearColor
   texture=/Game/Characters/Player/Male/Medium/Bodies/M_MED_SRDrift/Textures/T_ElectricNoise
  }
  n35 = Fresnel() {exponent=1}
  n36 = Multiply(A=NamedRerouteUsage(), B=Multiply(A=LinearInterpolate(Alpha=@n34.RGB) {constA=1, constB=12}, B=Power(Base=@n35) {constExponent=0.35}))
  n37 = Multiply(
   A=NamedRerouteDeclaration(None=@n3.R) {name=EmissiveArea, nodeColor=<rgba 0 1 0.0391 1>}
   B=Add(A=Multiply(A=@n24, B=@n28), B=Multiply(A=@n29, B=@n36))
  )
  n38 = Divide(A=@n37) {constB=1.5}
  n39 = MaterialFunctionCall(Color_Morning=@n37, Color_Day=@n37, Color_Evening=@n38, Color_Night=@n38) {materialFunction=&Color_Time_of_Day}
  n40 = NamedRerouteDeclaration(None=@n1.BaseColor) {name=BaseColorIn, nodeColor=<rgba 0.0312 1 0 1>}
  n41 = NamedRerouteUsage()
  FunctionOutput(None=SetMaterialAttributes!(
    MaterialAttributes=@n1.MaterialAttributes
    EmissiveColor=Add(A=@n2, B=@n39.Result)
    BaseColor=LinearInterpolate(A=@n40, Alpha=@n41) {constB=0}
   ) {attributeSetTypes [#1, #2]})
  MaterialExpressionTextureCoordinate_2 = TextureCoordinate()
  MaterialExpressionAdd_3 = Add(
   A=AppendVector(
    A=ScalarParameter() {group=Lightning, parameterName=Shape_UVOffsetX01, sortPriority=62}
    B=ScalarParameter() {group=Lightning, parameterName=Shape_UVOffsetY01, sortPriority=63}
   )
   B=Constant2Vector() {g=0.5, r=0.5}
  )
  MaterialExpressionScalarParameter_47 = ScalarParameter() {defaultValue=1.1, group=Lightning, parameterName=Shape_UVScale01, sortPriority=65}
  n42 = Custom!(OriginalUV=@MaterialExpressionTextureCoordinate_2, UVPoint=@MaterialExpressionAdd_3, UVScale=@MaterialExpressionScalarParameter_47) {
   code="float2 DistUV =OriginalUV-UVPoint;\r\nDistUV = DistUV/(UVScale/2) + float2(0.5,0.5);\r\n\r\nreturn DistUV;"
   description=CopyToPointUV
   outputType=CMOT_Float2
   inputs [
    {
     inputName=OriginalUV
     input {expression=&self:MaterialExpressionTextureCoordinate_2}
    }
    {
     inputName=UVPoint
     input {expression=&self:MaterialExpressionAdd_3}
    }
    {
     inputName=UVScale
     input {expression=&self:MaterialExpressionScalarParameter_47}
    }
   ]
  }
  NamedRerouteDeclaration(None=@n42) {name=PatternUV, nodeColor=<rgba 0.5703 0 1 1>}
  n43 = MaterialFunctionCall(Float3=@n17) {materialFunction=&BreakOutFloat3Components}
  n44 = Sine(None=MaterialFunctionCall(Divisor=@n19, Mod=Constant() {r=1}) {materialFunction=&MF_ModuloTime}.Clamped 0-1)
  NamedRerouteDeclaration(None=MaterialFunctionCall(A=@n43.R, B=@n43.R, C=@n43.R, Alpha=ConstantBiasScale(None=@n44)) {materialFunction=&Lerp_3Color}.Result) {name=RadiusEmissive, nodeColor=<rgba 0.3828 0 1 1>}
  Power(Base=@n17)
 ]}
