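# Character-style material for the Drift Trooper backpack, adding a lightning effect on top of the shared 2023 character parent shading.
# MF_CharacterParent_2023_PreFX builds material attributes, MF_DriftTrooperLightning modifies them using a tech-art mask texture, and MF_CharacterParent_2023_PostFX finishes the result.
# The lightning layer sits between the shared Pre and Post functions.
# Shading model comes from the graph, with usage flags for skeletal mesh, Nanite and mesh deformer.
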Material /Game/Accessories/FORT_Backpacks/Backpack_M_MED_DriftTrooper/Materials/M_DriftTrooper_Backpack_bak
graph {
 properties {
  bUseMaterialAttributes=true
  bUsedWithMeshDeformer=true
  bUsedWithNanite=true
  bUsedWithSkeletalMesh=true
  pixelDepthOffsetMode=PDOM_Legacy
  shadingModel=MSM_FromMaterialExpression
  usedShadingModels=Default Lit | Subsurface | Clear Coat | Subsurface Profile | Cloth
 }
 nodes [
  n1 = MaterialFunctionCall() {materialFunction=&MF_CharacterParent_2023_PreFX}
  n2 = TextureSampleParameter2D() {
   group=Lightning
   parameterName=TechArtMask_Lightning
   samplerType=SAMPLERTYPE_Masks
   texture=/Game/Accessories/FORT_Backpacks/Backpack_M_MED_DriftTrooper/Textures/T_DriftTrooper_Backpack_FX
  }
  MaterialFunctionCall(
   Material=MaterialFunctionCall!(Material=@n1.Result, Tech Art Mask In=@n2.RGB) {materialFunction=&/MF_DriftTrooperLightning}.Result
   M=@n1.M
   Pre Skinned Local Position=@n1.M
   Distance Blend=@n1.M
   Fresnel=@n1.M
  ) {materialFunction=&MF_CharacterParent_2023_PostFX}
 ]
}
